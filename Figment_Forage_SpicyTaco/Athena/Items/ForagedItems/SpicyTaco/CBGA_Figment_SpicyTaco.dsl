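; Spicy Taco foraged consumable that can spawn from Taco Time with limited availability.
; When from Taco Time, it resumes gravity simulation on begin, and the server kills it after a timer plus random delay, spawning despawn FX.
; A hit deactivates its trail.
; Failed consumption passes the player's account ID to a cosmetic handler, which stops that local player's search montage and shows the failure reason.
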(event Collision|EventActorBeginOverlap
  (params (OtherActor :type Actor%20Object%20Reference)))

(event EventTick
  (params (DeltaSeconds :type Float%20%28single-precision%29)))

(event EventOnAllowedToConsumeFailed
  (params (InteractingPawn :type Fort%20Pawn%20Object%20Reference))
  (Utilities|IsValid InteractingPawn
    (:Is%20Valid
      (bind _asfort_player_controller (Utilities|Casting|CastToFortPlayerController (Pawn|GetController InteractingPawn)))
      (CallFunction|HandleConsumeFailedVFX (Default|GetGameAccountId _asfort_player_controller)))))

(event Custom|HandleConsumeFailedVFX
  (params (NewParam :type Unique%20Net%20Id%20Structure))
  (Networking|CanExecuteCosmeticEvents
    (:True
      (bind _array_element (Utilities|Array|ForEachLoopwithBreak (Fortnite|GetLocalFortPlayerControllers true))
        (:LoopBody
          (if (== (Default|GetGameAccountId _array_element) NewParam)
            (Animation|StopAnimMontage (Game|Player|GetPlayerPawn _array_element) "/Game/Animation/Game/MainPlayer/Montages/General/Search.Search")
            (FailedAction|UnabletoPerformAction (Game|Player|GetPlayerPawn _array_element) true (Variables|FailedReason|GetFailedReason))))))))

(event EventBeginPlay
  (Variables|Gravity|SetResumeGravSimOnBegin (Variables|Default|GetFromTacoTime))
  (|CallParentFunction)
  (if (and (Networking|IsServer) (Variables|Default|GetFromTacoTime))
    ;[1 1 1 0.5] Accounts for limited availability from Taco Time
    (Utilities|Time|SetTimerbyEvent (delegate Custom|Kill) 6e+01)))

(event Custom|Kill
  ;[1 1 1 0.5] Accounts for limited availability from Taco Time
  (Utilities|FlowControl|Delay (Math|Random|RandomFloatinRange 0.0 2.0))
  (Variables|Default|SetHide false)
  (CallFunction|PlayFX)
  (Actor|DestroyActor))

(event Custom|PlayFX
  (Effects|SpawnEmitteratLocation (Variables|Default|GetDespawnFX) (Transformation|GetWorldLocation (Variables|Default|GetTrail)) (Transformation|GetActorRotation)))

(event Collision|EventHit
  (params (MyComp :type Primitive%20Component%20Object%20Reference) (Other :type Actor%20Object%20Reference) (OtherComp :type Primitive%20Component%20Object%20Reference) (bSelfMoved :type Boolean) (HitLocation :type Vector :default "0, 0, 0") (HitNormal :type Vector :default "0, 0, 0") (NormalImpulse :type Vector :default "0, 0, 0") (Hit :type Hit%20Result%20Structure%20%28by%20ref%29))
  (|CallParentFunction MyComp Other OtherComp bSelfMoved HitLocation HitNormal NormalImpulse Hit)
  (Components|Activation|Deactivate (Variables|Default|GetTrail)))

(fn OnInteractAllowedToConsume
  (params (InteractingPawn :type Fort%20Pawn%20Object%20Reference) (InteractionBeingAttempted :type EInteractionBeingAttempted%20Enum :default "FirstInteraction"))
  (Utilities|IsValid InteractingPawn
    (:Is%20Valid
      (bind _returnvalue (|CallParentFunction InteractingPawn))
      (return _returnvalue))
    (:Is%20Not%20Valid
      (return false))))

(fn OnRep_Hide
  (if (Variables|Default|GetHide)
    (Rendering|SetActorHiddenInGame true)))

(fn ConstructionScript
  (|CallParentFunction))
